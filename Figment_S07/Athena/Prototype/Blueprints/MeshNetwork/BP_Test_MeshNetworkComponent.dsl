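; Blueprint for a prototype mesh network test component, judging by its name and path, in the Figment Athena content.
; It declares a custom event, ServerCountElims, that takes a single integer parameter.
; The event has no body, so it performs no logic in this graph.
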(event Custom|ServerCountElims
  (params (NewParam :type Integer)))
